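# Physics asset for the LiverRomaine backpack cosmetic, giving its dangling handle a simple physics chain.
# Capsule bodies cover the pack, a small handle master bone and two handle segments.
# The master's swing and twist are locked to the pack, anchoring the chain.
# The first segment's limits allow swing on essentially one axis, while the second adds slight side swing and twist.
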PhysicsAsset /BRCosmetics/Accessories/FORT_Backpacks/Backpack_M_MED_LiverRomaine/Meshes/M_MED_LiverRomaine_Pack_Physics
bodies {
 dyn_handle_01 [{shapeType=Capsule, length=3.3, radius=3.3, shapeName=dyn_handle_01_capsule, center=<5 0 0>}]
 dyn_handle_02 [{shapeType=Capsule, length=3.3, radius=3.3, shapeName=dyn_handle_02_capsule, center=<6.3328 0 0>}]
 dyn_handle_c_MSTR [{shapeType=Capsule, length=0.505, radius=0.505, shapeName=dyn_handle_c_MSTR_capsule, rotation=<rot 0 -90 0>}]
 pack [{
   shapeType=Capsule
   length=12.0283
   radius=8.8696
   shapeName=pack_capsule
   center=<0.625 12.7007 -16.7249>
   rotation=<rot 0 -92.0214 -89.7382>
  }]
}
constraints [
 {
  bone1Name=dyn_handle_c_MSTR
  bone2Name=pack
  swing1LimitDegrees=45
  swing1Motion=Locked
  swing2LimitDegrees=45
  swing2Motion=Locked
  twistLimitDegrees=45
  twistMotion=Locked
 }
 {bone1Name=dyn_handle_01, bone2Name=dyn_handle_c_MSTR, swing1LimitDegrees=40, swing2LimitDegrees=0, twistLimitDegrees=0}
 {bone1Name=dyn_handle_02, bone2Name=dyn_handle_01, swing1LimitDegrees=40, swing2LimitDegrees=15, twistLimitDegrees=5}
]
