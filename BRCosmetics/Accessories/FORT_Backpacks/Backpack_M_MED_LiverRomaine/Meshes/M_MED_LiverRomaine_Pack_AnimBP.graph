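# Anim Blueprint for the LiverRomaine backpack cosmetic, adding secondary physics motion to its handle and envelope bones.
# The input pose takes handle AnimDynamics in the front end or RigidBody simulation otherwise, cached as FE Pose.
# When bRigidBodyEnabled is false, handle AnimDynamics runs on that cached pose, then seven chained envelope AnimDynamics nodes always follow.
# The update event sets a gravity override through a deprecated function and an inverted front-end flag.
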AnimBlueprint /BRCosmetics/Accessories/FORT_Backpacks/Backpack_M_MED_LiverRomaine/Meshes/M_MED_LiverRomaine_Pack_AnimBP
target_skeleton=&/M_MED_LiverRomaine_Pack_Skeleton
graphs [
 AnimGraph {nodes [
   n1 = Anim Dynamics - Bone: dyn_handle_01!(
    ComponentPose=Local To Component(LocalPose=Use cached pose 'FE Pose'())
    bAlphaBoolEnabled:bool=True
    GravityOverride:Vector="0,0,0"
    ExternalForce="0,0,0"
   ) -> (Pose) {
    alphaInputType=Bool
    angularDampingOverride=1
    angularSpringConstant=1500
    bAngularSpring=true
    bOverrideAngularDamping=true
    bUseGravityOverride=true
    alphaBoolBlend {blendOption=Linear}
    boundBone {boneName=dyn_handle_01}
    componentAppliedLinearAccClamp=<1000 1000 700>
    componentLinearAccScale=<1 1 0.6>
    componentLinearVelScale=<1 1 0.8>
    physicsBodyDefinitions [{
      boundBone {boneName=dyn_handle_01}
      boxExtents=<15 15 15>
      constraintSetup {angularLimitsMax=<0 0 20>, angularLimitsMin=<0 0 -40>, angularTarget=<1 0 0>}
      localJointOffset=<1.5 0 0>
     }]
   }
   n2 = Anim Dynamics - Bone: dyn_handle_02!(ComponentPose=@n1, bAlphaBoolEnabled:bool=True, GravityOverride:Vector="0,0,0", ExternalForce="0,0,0") -> (Pose) {
    alphaInputType=Bool
    angularBiasOverride=0.9999
    angularDampingOverride=0.999
    angularSpringConstant=1000
    bAngularSpring=true
    bLinearSpring=true
    bOverrideAngularBias=true
    bOverrideAngularDamping=true
    bUseGravityOverride=true
    linearSpringConstant=100
    alphaBoolBlend {blendOption=Linear}
    boundBone {boneName=dyn_handle_02}
    componentAppliedLinearAccClamp=<1000 1000 700>
    componentLinearAccScale=<1 1 0.6>
    componentLinearVelScale=<1 1 0.8>
    physicsBodyDefinitions [{
      boundBone {boneName=dyn_handle_02}
      boxExtents=<10 10 10>
      constraintSetup {angularLimitsMax=<5 10 20>, angularLimitsMin=<-5 -10 -40>, angularTarget=<1 0 0>}
      localJointOffset=<1 0 0>
     }]
   }
   n3 = Component To Local(ComponentPose=@n2)
   n4 = Use cached pose 'FE Pose'()
   n5 = Get bRigidBodyEnabled()
   n6 = Anim Dynamics - Bone: dyn_envelope_01(
    ComponentPose=Local To Component(LocalPose=Blend Poses by bool(BlendPose_0=@n4, BlendPose_1=@n3, bActiveValue=@n5) {blendType=Linear})
    Alpha=1
    ExternalForce="0,0,0"
   ) {
    angularDampingOverride=1
    angularSpringConstant=5000
    bAngularSpring=true
    bOverrideAngularDamping=true
    bUseGravityOverride=true
    alphaBoolBlend {blendOption=Linear}
    boundBone {boneName=dyn_envelope_01}
    componentAppliedLinearAccClamp=<500 500 500>
    componentLinearAccScale=<0.5 0.5 0.2>
    componentLinearVelScale=<0.7 0.7 0.3>
    gravityOverride {z=-500}
    physicsBodyDefinitions [{
      boundBone {boneName=dyn_envelope_01}
      boxExtents=<10 10 15>
      constraintSetup {angularLimitsMax=<10 0 10>, angularLimitsMin=<-10 0 -10>, angularTarget=<1 0 0>}
      localJointOffset=<2 0 0>
     }]
   }
   n7 = Anim Dynamics - Bone: dyn_envelope_02(ComponentPose=@n6, Alpha=1, ExternalForce="0,0,0") {
    angularDampingOverride=0.9999
    angularSpringConstant=2500
    bAngularSpring=true
    bOverrideAngularDamping=true
    bUseGravityOverride=true
    alphaBoolBlend {blendOption=Linear}
    boundBone {boneName=dyn_envelope_02}
    componentAppliedLinearAccClamp=<1000 1000 500>
    componentLinearAccScale=<0.5 0.5 0.2>
    componentLinearVelScale=<0.7 0.7 0.3>
    gravityOverride {z=-700}
    physicsBodyDefinitions [{
      boundBone {boneName=dyn_envelope_02}
      boxExtents=<10 10 10>
      constraintSetup {angularLimitsMax=<10 0 10>, angularLimitsMin=<-10 0 -10>, angularTarget=<1 0 0>}
      localJointOffset=<1 0 0>
     }]
   }
   n8 = Anim Dynamics - Bone: dyn_envelope_03(ComponentPose=@n7, Alpha=1, ExternalForce="0,0,0") {
    angularBiasOverride=0.9
    angularDampingOverride=1
    angularSpringConstant=7000
    bAngularSpring=true
    bOverrideAngularBias=true
    bOverrideAngularDamping=true
    bUseGravityOverride=true
    alphaBoolBlend {blendOption=Linear}
    boundBone {boneName=dyn_envelope_03}
    componentAppliedLinearAccClamp=<200 200 300>
    componentLinearAccScale=<0.5 0.5 0.2>
    componentLinearVelScale=<0.7 0.7 0.3>
    gravityOverride {z=-300}
    physicsBodyDefinitions [{
      boundBone {boneName=dyn_envelope_03}
      boxExtents=<10 15 15>
      constraintSetup {angularLimitsMax=<10 0 10>, angularLimitsMin=<-10 0 -10>, angularTarget=<1 0 0>}
      localJointOffset=<3 0 0>
     }]
   }
   n9 = Anim Dynamics - Bone: dyn_envelope_04(ComponentPose=@n8, Alpha=1, ExternalForce="0,0,0") {
    angularDampingOverride=0.9999
    angularSpringConstant=2500
    bAngularSpring=true
    bOverrideAngularDamping=true
    bUseGravityOverride=true
    alphaBoolBlend {blendOption=Linear}
    boundBone {boneName=dyn_envelope_04}
    componentAppliedLinearAccClamp=<1000 1000 500>
    componentLinearAccScale=<0.5 0.5 0.2>
    componentLinearVelScale=<0.7 0.7 0.3>
    gravityOverride {z=-700}
    physicsBodyDefinitions [{
      boundBone {boneName=dyn_envelope_04}
      boxExtents=<10 10 10>
      constraintSetup {angularLimitsMax=<10 0 10>, angularLimitsMin=<-10 0 -10>, angularTarget=<1 0 0>}
      localJointOffset=<1 0 0>
     }]
   }
   n10 = Anim Dynamics - Bone: dyn_envelope_05(ComponentPose=@n9, Alpha=1, ExternalForce="0,0,0") {
    angularBiasOverride=0.9
    angularDampingOverride=1
    angularSpringConstant=7000
    bAngularSpring=true
    bOverrideAngularBias=true
    bOverrideAngularDamping=true
    bUseGravityOverride=true
    alphaBoolBlend {blendOption=Linear}
    boundBone {boneName=dyn_envelope_05}
    componentAppliedLinearAccClamp=<200 200 300>
    componentLinearAccScale=<0.5 0.5 0.2>
    componentLinearVelScale=<0.7 0.7 0.3>
    gravityOverride {z=-300}
    physicsBodyDefinitions [{
      boundBone {boneName=dyn_envelope_05}
      boxExtents=<10 15 15>
      constraintSetup {angularLimitsMax=<10 0 10>, angularLimitsMin=<-10 0 -10>, angularTarget=<1 0 0>}
      localJointOffset=<3 0 0>
     }]
   }
   n11 = Anim Dynamics - Bone: dyn_envelope_06(ComponentPose=@n10, Alpha=1, ExternalForce="0,0,0") {
    angularDampingOverride=0.9999
    angularSpringConstant=2500
    bAngularSpring=true
    bOverrideAngularDamping=true
    bUseGravityOverride=true
    alphaBoolBlend {blendOption=Linear}
    boundBone {boneName=dyn_envelope_06}
    componentAppliedLinearAccClamp=<1000 1000 500>
    componentLinearAccScale=<0.5 0.5 0.2>
    componentLinearVelScale=<0.7 0.7 0.3>
    gravityOverride {z=-700}
    physicsBodyDefinitions [{
      boundBone {boneName=dyn_envelope_06}
      boxExtents=<10 10 10>
      constraintSetup {angularLimitsMax=<10 0 10>, angularLimitsMin=<-10 0 -10>, angularTarget=<1 0 0>}
      localJointOffset=<1 0 0>
     }]
   }
   n12 = Anim Dynamics - Bone: dyn_envelope_07(ComponentPose=@n11, Alpha=1, ExternalForce="0,0,0") {
    angularBiasOverride=0.9
    angularDampingOverride=1
    angularSpringConstant=7000
    bAngularSpring=true
    bOverrideAngularBias=true
    bOverrideAngularDamping=true
    bUseGravityOverride=true
    alphaBoolBlend {blendOption=Linear}
    boundBone {boneName=dyn_envelope_07}
    componentAppliedLinearAccClamp=<200 200 300>
    componentLinearAccScale=<0.5 0.5 0.2>
    componentLinearVelScale=<0.7 0.7 0.3>
    gravityOverride {z=-300}
    physicsBodyDefinitions [{
      boundBone {boneName=dyn_envelope_07}
      boxExtents=<10 15 15>
      constraintSetup {angularLimitsMax=<10 0 10>, angularLimitsMin=<-10 0 -10>, angularTarget=<1 0 0>}
      localJointOffset=<3 0 0>
     }]
   }
   Output Pose(Result=Component To Local(ComponentPose=@n12)) {layerGroup=None}
   n13 = Anim Dynamics - Bone: dyn_handle_01!(ComponentPose=Local To Component(LocalPose=Use cached pose 'Start Pose'()), Alpha:float=1, GravityOverride="0,0,0", ExternalForce="0,0,0") -> (Pose) {
    angularDampingOverride=1
    angularSpringConstant=1500
    bAngularSpring=true
    bOverrideAngularDamping=true
    bUseGravityOverride=true
    alphaBoolBlend {blendOption=Linear}
    boundBone {boneName=dyn_handle_01}
    componentAppliedLinearAccClamp=<1000 1000 700>
    componentLinearAccScale=<1 1 0.6>
    componentLinearVelScale=<1 1 0.8>
    physicsBodyDefinitions [{
      boundBone {boneName=dyn_handle_01}
      boxExtents=<15 15 15>
      constraintSetup {angularLimitsMax=<0 0 20>, angularLimitsMin=<0 0 -40>, angularTarget=<1 0 0>}
      localJointOffset=<1.5 0 0>
     }]
   }
   n14 = Anim Dynamics - Bone: dyn_handle_02!(ComponentPose=@n13, Alpha:float=1, GravityOverride="0,0,0", ExternalForce="0,0,0") -> (Pose) {
    angularBiasOverride=0.9999
    angularDampingOverride=0.999
    angularSpringConstant=1000
    bAngularSpring=true
    bLinearSpring=true
    bOverrideAngularBias=true
    bOverrideAngularDamping=true
    bUseGravityOverride=true
    linearSpringConstant=100
    alphaBoolBlend {blendOption=Linear}
    boundBone {boneName=dyn_handle_02}
    componentAppliedLinearAccClamp=<1000 1000 700>
    componentLinearAccScale=<1 1 0.6>
    componentLinearVelScale=<1 1 0.8>
    physicsBodyDefinitions [{
      boundBone {boneName=dyn_handle_02}
      boxExtents=<10 10 10>
      constraintSetup {angularLimitsMax=<5 5 20>, angularLimitsMin=<-5 -5 -40>, angularTarget=<1 0 0>}
      localJointOffset=<1 0 0>
     }]
   }
   n15 = RigidBody!(ComponentPose=Local To Component(LocalPose=Use cached pose 'Start Pose'()), Alpha, OverrideWorldGravity:Vector="0,0,0", ExternalForce) -> (Pose) {
    bClampLinearTranslationLimitToRefPose=true
    bOverrideWorldGravity=true
    overridePhysicsAsset=&/M_MED_LiverRomaine_Pack_Physics
    simulationSpace=BaseBoneSpace
    alphaBoolBlend {blendOption=Linear}
    baseBoneRef {boneName=pack}
    simSpaceSettings {
     dampingAlpha=0
     maxAngularAcceleration=10
     maxAngularVelocity=2
     maxLinearAcceleration=1000
     maxLinearVelocity=500
     worldAlpha=1
     externalLinearDragV=<0.5 0.5>
    }
   }
   Save cached pose 'FE Pose'(Pose=Blend Poses by bool(BlendPose_0=Component To Local(ComponentPose=@n14), BlendPose_1=Component To Local(ComponentPose=@n15), bActiveValue=Get bIsInFrontEnd()) {blendType=Linear}) {cacheName=FE Pose}
   Save cached pose 'Start Pose'(Pose=Input Pose()) {cacheName=Start Pose}
  ]}
 EventGraph {nodes [
   Event BlueprintUpdateAnimation()
   > n1 = Gravity Override State Function DEPRECATED(CharacterPartAnimInstance=Self-Reference(), GameTimeInSeconds=GetGameTimeInSeconds(), GravityOverrideParams=Get Gravity Override Params())
   > Set Gravity Override(Gravity Override=@n1)
   > Set IsNOTFrontEnd(IsNOTFrontEnd:bool=NOT Boolean(A=Get bIsInFrontEnd())) -> (Output_Get:bool)
  ]}
]
variables [
 Gravity Override {type=Vector}
 Gravity Override Params {type=GravityOverrideParamsStruct}
 IsNOTFrontEnd {type=boolean}
]
